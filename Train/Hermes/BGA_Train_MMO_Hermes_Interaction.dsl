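; Interaction actor Blueprint for the Hermes train, mostly a stub that only defines where its interaction icon sits.
; GetIconPlacement returns the given actor's location raised 50 units on Z, paired with a zero second value, and ignores the viewing actor.
; BeginPlay, overlap and Tick events hold no logic, and the construction script only calls its parent.
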(event EventBeginPlay
  (|CallParentFunction) ; This node is disabled and will not be called.\nDrag off pins to build functionality.
)

(event Collision|EventActorBeginOverlap
  (params (OtherActor :type Actor%20Object%20Reference)))

(event EventTick
  (params (DeltaSeconds :type Float%20%28single-precision%29)))

(fn GetIconPlacement
  (params (SelfActor :type Actor%20Object%20Reference) (ViewingActor :type Actor%20Object%20Reference))
  (return (+ (Transformation|GetActorLocation SelfActor) "0.000000,0.000000,50.000000") "0, 0, 0"))

(fn ConstructionScript
  (|CallParentFunction))
